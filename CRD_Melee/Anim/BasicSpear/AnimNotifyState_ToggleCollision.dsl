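; Anim notify state for the BasicSpear melee animations that lets a player pawn pass through other pawns during the notify window.
; On begin, it casts the mesh owner to FortPlayerPawn and sets its capsule to ignore the Pawn and GameTraceChannel4 channels.
; On end, it sets both channels to Block rather than restoring saved responses.
; Owners that are not FortPlayerPawns are left untouched.
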(fn GetNotifyName
  (return "AnimNotifyState_ToggleCollision_AI_And_Pawn"))

(fn Received_NotifyBegin
  (params (MeshComp :type Skeletal%20Mesh%20Component%20Object%20Reference) (Animation :type Anim%20Sequence%20Base%20Object%20Reference) (TotalDuration :type Float%20%28single-precision%29) (EventReference :type Anim%20Notify%20Event%20Reference%20Structure%20%28by%20ref%29))
  (bind _asfort_player_pawn (Utilities|Casting|CastToFortPlayerPawn (Components|GetOwner MeshComp))
    (:then
      (Collision|SetCollisionResponsetoChannel (Variables|/Script/Engine.Character|GetCapsuleComponent _asfort_player_pawn) "ECC_GameTraceChannel4" "ECR_Ignore")
      (Collision|SetCollisionResponsetoChannel (Variables|/Script/Engine.Character|GetCapsuleComponent _asfort_player_pawn) "ECC_Pawn" "ECR_Ignore")
      (return true))
    (:CastFailed
      (return false))))

(fn Received_NotifyEnd
  (params (MeshComp :type Skeletal%20Mesh%20Component%20Object%20Reference) (Animation :type Anim%20Sequence%20Base%20Object%20Reference) (EventReference :type Anim%20Notify%20Event%20Reference%20Structure%20%28by%20ref%29))
  (bind _asfort_player_pawn (Utilities|Casting|CastToFortPlayerPawn (Components|GetOwner MeshComp))
    (:then
      (Collision|SetCollisionResponsetoChannel (Variables|/Script/Engine.Character|GetCapsuleComponent _asfort_player_pawn) "ECC_GameTraceChannel4" "ECR_Block")
      (Collision|SetCollisionResponsetoChannel (Variables|/Script/Engine.Character|GetCapsuleComponent _asfort_player_pawn) "ECC_Pawn" "ECR_Block")
      (return true))
    (:CastFailed
      (return false))))
